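; SouthBulk AI ground-wave slam actor: an expanding donut-shaped shockwave that damages what it hits.
; On spawn it shakes cameras, spawns a burst effect and reads its scalable duration.
; Donut attack size updates drive both ring meshes, a sine-modulated dynamic material opacity, the audio cylinder and the Niagara damage-area effect.
; Hits outside the ignore list receive a damage gameplay effect through the owning pawn's ability system, and the actor destroys itself when duration ends.
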(event EventBeginPlay
  (Camera|PlayWorldCameraShake "/SouthBulkAI/Gameplay/B_GroundWave_CameraShake.B_GroundWave_CameraShake_C" (Transformation|GetActorLocation) (ScalableFloat|GetValueatLevel (Variables|Default|GetSF_CameraShakeRadius)) (ScalableFloat|GetValueatLevel (Variables|Default|GetSF_CameraShakeRadius)))
  (Variables|Internal|SetGroundWaveDuration (ScalableFloat|GetValueatLevelwithContextParam (Actor|GetOwner) (Variables|Default|GetSF_GroundWaveDuration)))
  (Niagara|SpawnSystematLocation "/SouthBulkAI/Assets/VFX/Slam/NS_SouthBulk_SlamBurst.NS_SouthBulk_SlamBurst" (Transformation|GetActorLocation)))

(event Custom|HandleShockwaveDamage
  (params (TargetActor :type Actor%20Object%20Reference))
  (bind _asfort_pawn (Utilities|Casting|CastToFortPawn TargetActor)
    (:then
      (if (GameplayTags|HasAnyMatchingGameplayTags _asfort_pawn (Variables|Internal|GetPawnShockwaveExclusionTags))
        _
        (else
          _)))
    (:CastFailed
      _))
  (CallFunction|ApplyGroundWaveDamage TargetActor))

(event OnActorHit(FortDonutAttack)
  (params (HitActor :type Actor%20Object%20Reference))
  (if (not (Utilities|Array|ContainsItem (Variables|Internal|GetWaveIgnoreActors) HitActor))
    (CallFunction|HandleShockwaveDamage HitActor)))

(event OnDurationEnd(FortDonutAttack)
  (Rendering|SetVisibility (Variables|Default|GetSM_GroundWave_Outter_Ring) false false)
  (Actor|DestroyActor))

(event OnSizeChange(FortDonutAttack)
  (params (OuterRadius :type Float%20%28single-precision%29) (InnerRadius :type Float%20%28single-precision%29) (Height :type Float%20%28single-precision%29) (DurationNormalized :type Float%20%28single-precision%29))
  (bind _returnvalue (/ Height 1e+02))
  (bind _returnvalue_2 (/ OuterRadius 1e+02))
  (Transformation|SetWorldScale3D (Variables|Default|GetSM_GroundWave_Outter_Ring) (Math|Vector|MakeVector _returnvalue_2 _returnvalue_2 _returnvalue))
  (Transformation|SetWorldScale3D (Variables|Default|GetSM_Spire_Inner_Ring) (Math|Vector|MakeVector (/ InnerRadius 1e+02) (/ InnerRadius 1e+02) _returnvalue))
  (Rendering|Material|SetScalarParameterValue (Variables|Default|GetShockwaveMaterial) "OpacityMult" (Math|Trig|Sin(Radians) (* DurationNormalized (Variables|Internal|GetGroundWaveDuration))) :!class /Script/Engine.MaterialInstanceDynamic)
  (Control|SetRadius (Variables|Default|GetAudioShapeCylinder) OuterRadius :!class /Script/AudioShapes.AudioShapeCylinderComponent)
  (Niagara|SetNiagaraVariable(Float) (Variables|Default|GetFX_SlamDamageArea) "TargetWaveSize" _returnvalue_2)
  (Utilities|FlowControl|DoOnce false
    (:Completed
      (Components|Activation|SetActive (Variables|Default|GetFX_SlamDamageArea) true true))))

(fn ApplyGroundWaveDamage
  (params (EffectTargetActor :type Actor%20Object%20Reference))
  (bind _asfort_pawn (Utilities|Casting|CastToFortPawn (Actor|GetOwner)))
  (Fortnite|ApplyGameplayEffectsToActor EffectTargetActor (Utilities|Array|MakeArray (Variables|Default|GetGE_GroundWaveDamage)) (Transformation|GetActorLocation _asfort_pawn) false 0 (Ability|GetAbilitySystemComponent _asfort_pawn))
  (return 0))

(fn ConstructionScript
  (bind _returnvalue (Rendering|Material|CreateDynamicMaterialInstance "/SouthBulkAI/Assets/VFX/MI_Spire_Shockwave_Gradient_Glow.MI_Spire_Shockwave_Gradient_Glow" "None" "None" :!class /Script/Engine.KismetMaterialLibrary))
  (bind _output_get (Variables|Default|SetShockwaveMaterial _returnvalue))
  (Rendering|Material|SetMaterial (Variables|Default|GetSM_GroundWave_Outter_Ring) 0 _output_get))
