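; Damage actor for the SouthBalk AI V2 big beam, applying the drone laser damage GameplayEffect along a forward, downward-angled line.
; On authority, BeginPlay starts a looping timer at a ScalableFloat interval; SetBeamDamageDuration clears it after a delay.
; Each tick sphere-traces for objects, ignoring itself and its instigator, and damages every hit with an AbilitySystemComponent.
; The trace runs only where cosmetic events cannot execute, such as dedicated servers.
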(event EventBeginPlay
  (Networking|SwitchHasAuthority
    (:Authority
      (bind _returnvalue (Utilities|Time|SetTimerbyEvent (EventDispatchers|CreateEvent :!event Custom|StartTargetting) (ScalableFloat|GetValueatLevel (Variables|Default|GetTargettingInterval)) true))
      (Variables|InternalData|SetHNDL_TargettingInterval _returnvalue))))

(event Custom|SetBeamDamageDuration
  (params (Duration :type Float%20%28double-precision%29))
  (Utilities|FlowControl|Delay Duration)
  (Utilities|Time|ClearandInvalidateTimerbyHandle (Variables|InternalData|GetHNDL_TargettingInterval)))

(event Custom|StartTargetting
  (Networking|CanExecuteCosmeticEvents
    (:False
      (bind _outhits (Collision|MultiSphereTraceForObjects (Transformation|GetActorLocation self) (Utilities|Operators|Add (Transformation|GetActorLocation self) (* (Transformation|GetActorForwardVector self) 8.5e+02) (* (Transformation|GetActorUpVector self) -4.5e+02)) 5e+01 (Variables|Default|GetTraceObjectTypes) false (Utilities|Array|MakeArray self (Variables|Actor|GetInstigator))))
      (for _array_element _outhits
        (CallFunction|HandleDamage (.HitActor (Collision|BreakHitResult _array_element)))))))

(event Custom|HandleDamage
  (params (Actor :type Actor%20Object%20Reference))
  (bind _returnvalue (Ability|GetAbilitySystemComponent Actor))
  (Utilities|IsValid _returnvalue
    (:Is%20Valid
      (GameplayEffects|ApplyGameplayEffectToTarget :self (Ability|GetAbilitySystemComponent 0) :GameplayEffectClass "/SouthBulkAI/Abilities/GE_SouthBalkAI_DroneLazer_Dmg.GE_SouthBalkAI_DroneLazer_Dmg_C" :Target _returnvalue :Level 0.0))))

(fn ConstructionScript)
