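; SouthBulk AI projectile that explodes on hit, server stop or water entry, with a looping sound indicator in flight.
; Hit and stop share an Exploded guard, spawning Niagara impact VFX before DoExplosion; hits adjust it for pawns, water and trains.
; Server stops also trigger ForceFX: radial force and screen cue, gated by platform and detail level.
; Explosion makes AI noise, plays player-specific or default impact audio and destroys the target projection.
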(event EventBeginPlay
  ; Begin Play
  (SoundIndicator|StartSoundIndicatorLooping (Variables|Default|GetFortSoundIndicator_ProjectileAV) (Transformation|GetActorLocation) 0 1.0 0))

(event EventOnStoppedOnServer
  (bind _returnvalue (Transformation|GetActorLocation))
  ;[1 1 1] EVENT -> OnStoppedOnServer
  (if (not (Variables|Default|GetExploded))
    (Variables|Default|SetExploded true)
    (Utilities|FlowControl|Delay 0.033)
    (bind _returnvalue_2 (Niagara|SpawnSystematLocation (Variables|Default|GetVFX_Impact) _returnvalue (Transformation|GetActorRotation) (Variables|Default|GetVFX_Impact_Scale)))
    (Variables|Default|SetVFXImpactRef _returnvalue_2)
    (CallFunction|ForceFX _returnvalue)
    (Collision|SetCollisionEnabled (Variables|Collision|GetCapsuleComponent) "NoCollision")
    (Components|Activation|Deactivate (Variables|Default|GetNS_Projectile))
    (Delivery|DoExplosion)))

(event Collision|EventHit
  (params (MyComp :type Primitive%20Component%20Object%20Reference) (Other :type Actor%20Object%20Reference) (OtherComp :type Primitive%20Component%20Object%20Reference) (bSelfMoved :type Boolean) (HitLocation :type Vector :default "0, 0, 0") (HitNormal :type Vector :default "0, 0, 0") (NormalImpulse :type Vector :default "0, 0, 0") (Hit :type Hit%20Result%20Structure%20%28by%20ref%29))
  (bind (_bblockinghit _binitialoverlap _time _distance _location _impactpoint _normal _impactnormal _physmat _hitactor) (Collision|BreakHitResult Hit))
  (bind (_asfort_pawn _bsuccess) (Utilities|Casting|CastToFortPawn _hitactor))
  (bind (_asfort_water_body_actor _bsuccess_2) (Utilities|Casting|CastToFortWaterBodyActor _hitactor))
  (bind _vfximpactref (Variables|Default|GetVFXImpactRef))
  ;[1 1 1] EVENT -> Hit
  (if (not (Variables|Default|GetExploded))
    (Variables|Default|SetExploded true)
    (bind _returnvalue (Niagara|SpawnSystematLocation (Variables|Default|GetVFX_Impact) HitLocation (Math|Rotator|MakeRotfromX HitNormal) (Variables|Default|GetVFX_Impact_Scale)))
    (Variables|Default|SetVFXImpactRef _returnvalue)
    (Niagara|SetNiagaraVariable(Bool) _returnvalue "bSpawnDecal" (Widget|UEFN|InvertBool _bsuccess))
    (Niagara|SetNiagaraVariable(Bool) _returnvalue "bIsInWater" _bsuccess_2)
    (Game|Components|ProjectileMovement|StopSimulating (Variables|Movement|GetProjectileMovementComponent) Hit)
    (Collision|SetCollisionEnabled (Variables|Collision|GetCapsuleComponent) "NoCollision")
    ;;[0.9427 0.1768 0.1768 0.5] VFX: If hit actor = Train, Decal won't show up.
    (Components|Activation|Deactivate (Variables|Default|GetNS_Projectile))
    (bind _returnvalue_2 (GameplayTags|HasMatchingGameplayTag(Message) _hitactor "(TagName=\"Athena.Gameplay.Train\")"))
    (Utilities|IsValid _vfximpactref
      (:Is%20Valid
        (Niagara|SetNiagaraVariable(Bool) _vfximpactref "IsTrain" _returnvalue_2))
      (:Is%20Not%20Valid
        _))
    (Delivery|DoExplosion)
    (SoundIndicator|StopSoundIndicatorLooping (Variables|Default|GetFortSoundIndicator_ProjectileAV))))

(event EventOnExploded
  (params (HitActors :type Array%20of%20Actor%20Object%20References) (HitResults :type Array%20of%20Hit%20Result%20Structures))
  (bind (_asfort_player_pawn _bsuccess) (Utilities|Casting|CastToFortPlayerPawn (Utilities|Array|Get(acopy) HitActors)))
  ;[1 1 1] EVENT -> OnExploded
  (AI|MakeNoiseAtAIProjectile
    (:Then
      (Networking|CanExecuteCosmeticEvents
        (:True
          (if (Utilities|Array|IsValidIndex HitActors)
            (Variables|Default|SetProjectileImpact_Cue (select _bsuccess (Variables|Default|GetProjectile%20Impact%20Cue%20Player) (Variables|Default|GetProjectile%20Impact%20Cue%20Default)))
            (else
              (Variables|Default|SetProjectileImpact_Cue (Variables|Default|GetProjectile%20Impact%20Cue%20Default))))
          (Audio|PlaySoundatLocation (Variables|Default|GetProjectileImpact_Cue) (Transformation|GetWorldLocation (Variables|Collision|GetCapsuleComponent))))
        (:False
          _))
      (Actor|DestroyActor (Variables|Default|GetTargetProjection)))))

(event OnEnterWaterEvent(WaterInteractionComponent)
  (params (WaterBody :type Fort%20Water%20Body%20Actor%20Object%20Reference) (WaterInteractionComponent :type Fort%20Water%20Interaction%20Component%20Object%20Reference) (bIsFirstBody :type Boolean))
  ; Water Events
  (Networking|SwitchHasAuthority
    (:Authority
      (Delivery|DoExplosion))
    (:Remote
      (Audio|PlaySoundatLocation (Variables|Default|GetSplash%20Sound) (Transformation|GetActorLocation))
      (Components|Activation|SetActive (Variables|Default|GetNS_Projectile) false false)
      (Effects|Components|ParticleSystem|SpawnFXAtLocation "/Game/Weapons/FORT_Generic_Effects/Impacts/Water/NS_Impact_Water_Athena/NS_Impact_Water_Apollo.NS_Impact_Water_Apollo" (Transformation|GetWorldLocation (Variables|Collision|GetCapsuleComponent))))))

(event OnProjectileStop(ProjectileMovementComponent)
  (params (ImpactResult :type Hit%20Result%20Structure%20%28by%20ref%29))
  (Audio|Components|Audio|FadeOut (Variables|ProjectileDeimosRanged|GetRanged_Projectile_Sound) 0.2 0.0 "Linear" :!class /Script/Engine.AudioComponent))

(fn ForceFX
  (params (Location :type Vector))
  (if (Math|Boolean|ANDBoolean (not (Fortnite|IsMobilePlatform)) (not (Class|PlatformHelpersLibrary|IsPcPerfMode)) (> (Rendering|GetRenderingDetailMode) 1))
    (Game|SpawnActorfromClass "/Game/Effects/Niagara/Blueprints/PhsyicsFields/FN_RadialForce.FN_RadialForce_C" (Math|Conversions|ToTransform(Vector) Location) "Undefined" "MultiplyWithRoot" 0 3.0 0.6 1e+03 1.5 0.05 0)
    (GameplayCue|ExecuteGameplayCue%28Burst%29Local self (Variables|Default|GetGCN_ScreenFX) (Ability|GameplayCue|MakeGameplayCueParameters 0.0)) ; Cam shake + controller vibration
  ))

(fn ConstructionScript)
